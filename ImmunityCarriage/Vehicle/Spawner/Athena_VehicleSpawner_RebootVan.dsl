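; Reboot Van vehicle spawner for Athena (Battle Royale) that spawns the van at the spawner's transform, except in playlists carrying the solo tag.
; Both the normal and static spawn events call their parent, then continue only on authority.
; Each validates its own vehicle class, merges the normal and static class paths, checks the game state's playlist tag, and calls SpawnVehiclewithConstruction.
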(event EventSpawnVehicle
  (bind _returnvalue (CallFunction|GetVehicleClass))
  (|CallParentFunction)
  (Networking|SwitchHasAuthority
    (:Authority
      (if (Utilities|IsValidClass _returnvalue)
        (bind _data (Utilities|FlowControl|MergeExecutionwithData _returnvalue (CallFunction|GetStaticVehicleClass))
          (:Then
            (if (not (Class|FortGameStateAthena|PlaylistHasTag (Fortnite|Athena|GameState|GetGameStateAthena) (Variables|Default|GetSoloTag)))
              (CallFunction|SpawnVehiclewithConstruction _data (Transformation|GetActorTransform)))))))))

(event EventSpawnStaticVehicle
  (bind _returnvalue (CallFunction|GetStaticVehicleClass))
  (|CallParentFunction)
  (Networking|SwitchHasAuthority
    (:Authority
      (if (Utilities|IsValidClass _returnvalue)
        (bind _data (Utilities|FlowControl|MergeExecutionwithData (CallFunction|GetVehicleClass) _returnvalue)
          (:Then
            (if (not (Class|FortGameStateAthena|PlaylistHasTag (Fortnite|Athena|GameState|GetGameStateAthena) (Variables|Default|GetSoloTag)))
              (CallFunction|SpawnVehiclewithConstruction _data (Transformation|GetActorTransform)))))))))

(fn ConstructionScript
  (|CallParentFunction))
